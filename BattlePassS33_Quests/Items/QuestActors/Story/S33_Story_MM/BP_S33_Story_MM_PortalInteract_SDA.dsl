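; Season 33 story quest actor that pairs with a portal actor for a portal interact step.
; The FortLinkToActor connect event stores the linked actor, and its destroyed event destroys this actor, both printing debug messages.
; FindPortalActor, not called within this file, optionally sphere-overlaps nearby actors, stores the last one with the quest's portal tag and returns success.
; The side effect and Tick events are empty.
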(event EventBeginPlay
  (|CallParentFunction))

(event EventTick
  (params (DeltaSeconds :type Float%20%28single-precision%29)))

(event OnLinkedActorConnected(FortLinkToActor)
  (params (LinkedActor :type Actor%20Object%20Reference))
  (Variables|Default|SetLinkedActor LinkedActor)
  (Development|PrintString "Portal Actor Connected via Link!" true true "(R=1.000000,G=0.533333,B=0.000000,A=1.000000)" 5.0))

(event OnLinkedActorDestroyed(FortLinkToActor)
  (params (DamageCauser :type Actor%20Object%20Reference))
  (Development|PrintString "Portal Actor Destroyed!" true true "(R=1.000000,G=0.000000,B=0.000000,A=1.000000)" 5.0)
  (Actor|DestroyActor))

(event SideEffects|EventBP_ApplySideEffect
  (params (ReactionSource :type Object%20Reference) (TargetObject :type Object%20Reference) (SideEffectTriggerTag :type Gameplay%20Tag%20Structure%20%28by%20ref%29) (SideEffectTriggerType :type EBlueprintableSideEffectTriggerType%20Enum%20%28by%20ref%29 :default "TriggeredAsTarget")))

(fn FindPortalActor
  (params (Enable :type Boolean))
  (if Enable
    (Variables|Default|SetSuccess false)
    (bind _outactors (Collision|SphereOverlapActors (Transformation|GetActorLocation) 7.5e+02 (Utilities|Array|MakeArray "ObjectTypeQuery1")))
    (for _array_element _outactors
      (if (Actor|ActorHasTag _array_element "Athena.Quests.S33.Story.MM.02.PortalInteract.Actor")
        (Variables|Default|SetLinkedActor _array_element)
        (Variables|Default|SetSuccess true)))
    (return (Variables|Default|GetSuccess))
    (else
      (return false))))

(fn ConstructionScript)
