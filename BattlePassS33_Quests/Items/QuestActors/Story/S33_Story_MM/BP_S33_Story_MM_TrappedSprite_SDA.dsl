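; Trapped Sprite actor for a Season 33 story quest, which removes itself when its linked cage actor is destroyed.
; It responds to FortLinkToActor events, printing a red debug message when the linked actor connects or is destroyed, then destroying itself in the latter case.
; BeginPlay only calls the parent, while overlap, tick and construction script are empty.
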(event EventBeginPlay
  (|CallParentFunction))

(event Collision|EventActorBeginOverlap
  (params (OtherActor :type Actor%20Object%20Reference)))

(event EventTick
  (params (DeltaSeconds :type Float%20%28single-precision%29)))

(event OnLinkedActorConnected(FortLinkToActor)
  (params (LinkedActor :type Actor%20Object%20Reference))
  (Development|PrintString "Trapped Sprite Cage Actor Connected!" true true "(R=1.000000,G=0.000000,B=0.000000,A=1.000000)" 5.0))

(event OnLinkedActorDestroyed(FortLinkToActor)
  (params (DamageCauser :type Actor%20Object%20Reference))
  (Development|PrintString "Trapped Sprite Cage Actor Destroyed!" true true "(R=1.000000,G=0.000000,B=0.000000,A=1.000000)" 5.0)
  (Actor|DestroyActor))

(fn ConstructionScript)
